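; Dynamic UI director for the AI progress bar that tracks challenge manager changes on the local player's view target.
; On BeginPlay it listens to the HUD context for view target changes, unbinding on EndPlay.
; Each change rebinds a cached BP_AIProgressBarPawnComponent's challenge manager event and calls the handler.
; That handler's touch/mobile branch is left unconnected because UI extension slots add the widget; the director remains for a possible dynamic UI switch.
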(event EventBeginPlay
  (bind _returnvalue (BlueprintContext|GetFortHUDContext))
  (Utilities|IsValid _returnvalue
    (:Is%20Valid
      (HUD|AssignOnLocalPlayerViewTargetChanged _returnvalue (EventDispatchers|CreateEvent :!event OnLocalPlayerViewTargetChanged))
      (CallFunction|OnLocalPlayerViewTargetChanged))))

(event EventEndPlay
  (params (EndPlayReason :type EEndPlayReason%20Enum :default "Destroyed"))
  (bind _returnvalue (BlueprintContext|GetFortHUDContext))
  (Utilities|IsValid _returnvalue
    (:Is%20Valid
      (HUD|UnbindEventfromOnLocalPlayerViewTargetChanged _returnvalue (EventDispatchers|CreateEvent :!event OnLocalPlayerViewTargetChanged)))))

(fn HandleChallengeManagerChanged
  ; This is disconnected because we are adding the widget through the UI extension slots. We are leaving the director here in case we want to change to dynamic UI
  (if (and (TouchPlatform|IsUsingTouch (DynamicUIDirector|GetOwningLocalPlayerController)) (TouchPlatform|IsMobileGame))
    _))

(fn OnLocalPlayerViewTargetChanged
  (bind _returnvalue (DynamicUIDirector|GetOwningLocalPlayerController))
  (bind _returnvalue_2 (Actor|GetComponentbyClass (Pawn|GetViewTarget _returnvalue) "/AIProgressBar/Blueprints/Components/BP_AIProgressBarPawnComponent.BP_AIProgressBarPawnComponent_C"))
  (bind _returnvalue_3 (Pawn|GetViewTarget _returnvalue))
  (bind _cachedpawncomponent (Variables|Default|GetCachedPawnComponent))
  (Utilities|FlowControl|Sequence
    (:then_0
      (Utilities|IsValid _cachedpawncomponent
        (:Is%20Valid
          (Default|UnbindEventfromOnCurrentChallengeManagerChanged _cachedpawncomponent (EventDispatchers|CreateEvent :!event Handle%20Challenge%20Manager%20Changed))
          (Variables|Default|SetCachedPawnComponent 0))))
    (:then_1
      (Utilities|IsValid _returnvalue_3
        (:Is%20Valid
          (Utilities|IsValid _returnvalue_2
            (:Is%20Valid
              (bind _output_get (Variables|Default|SetCachedPawnComponent _returnvalue_2)) ; Cache it so we can unbind from it later
              (Default|BindEventtoOnCurrentChallengeManagerChanged _output_get (EventDispatchers|CreateEvent :!event Handle%20Challenge%20Manager%20Changed)))))))
    (:then_2
      (CallFunction|HandleChallengeManagerChanged))))

(fn ConstructionScript)
