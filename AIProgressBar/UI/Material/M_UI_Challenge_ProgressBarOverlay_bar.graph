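# Translucent UI material for the bar overlay of the challenge progress bar.
# It samples the progress frame bar texture and isolates one channel through the TextureToUse channel mask parameter.
# That mask is multiplied by the alpha of the OverlayColor parameter, which defaults to opaque white, and is also passed on through a reroute.
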Material /AIProgressBar/UI/Material/M_UI_Challenge_ProgressBarOverlay_bar
graph {
 properties {blendMode=BLEND_Translucent, materialDomain=MD_UI, pixelDepthOffsetMode=PDOM_Legacy, usedShadingModels=Default Lit}
 nodes [
  n1 = ChannelMaskParameter(None=TextureSample() {texture=/AIProgressBar/UI/Texture/T_UI_Challenge_ProgressFrame_Bar}.RGB) {parameterName=TextureToUse}
  Multiply(A=VectorParameter() {parameterName=OverlayColor, defaultValue=<rgba 1 1 1 1>}.A, B=@n1.R)
  Reroute(None=@n1.R)
 ]
}
